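# Sound cue for the looping sound of a broken security generator gadget in Athena.
# A single looping wave player feeds the cue output with the broken loop wave at slightly reduced volume.
# It uses the electric generator attenuation and the Athena ambient sound class.
# Concurrency overrides allow one instance per owner and prevent new ones from starting.
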SoundCue /Game/Athena/Sounds/Gadgets/Generator/Security_Generator_Broken_Loop_Cue
root=WavePlayer_2
nodes [{id=WavePlayer_2, properties {bLooping=true, soundWaveAssetPtr=&/Security_Generator_Broken_Loop}}]
properties {
 attenuationSettings=&/Electric_Generator_Att
 maxDistance=900
 soundClassObject=&Fort_Ambient_Athena
 volumeMultiplier=0.832
 concurrencyOverrides {bLimitToOwner=true, maxCount=1, resolutionRule=PreventNew}
}
