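# Sound cue for impacts on the Athena Security Generator gadget, built from demolition metal impact waves.
# A Mixer layers a random metal garage impact at full volume with a random metal machine impact at half volume, then feeds a Modulator at the root for variation.
# It plays in the impacts sound class with small-object demolition attenuation and overrides concurrency to prevent new instances once the limit is reached.
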SoundCue /Game/Athena/Sounds/Gadgets/Generator/Security_Generator_Impact_Cue
root=Modulator_1
nodes [
 {id=Mixer_2, children [Random_2,Random_1], properties {inputVolume [1,0.5]}}
 {id=Modulator_1, children [Mixer_2]}
 {id=Random_1, children [WavePlayer_5,WavePlayer_7,WavePlayer_1], properties {weights [1,1,1]}}
 {id=Random_2, children [WavePlayer_8, WavePlayer_9, WavePlayer_10, WavePlayer_11], properties {weights [1,1,1,1]}}
 {id=WavePlayer_1, properties {soundWaveAssetPtr=/Game/Sounds/Fort_Impact_Sounds/Demolition/Metal/Metal_Machine_Impact05}}
 {id=WavePlayer_10, properties {soundWaveAssetPtr=/Game/Sounds/Fort_Impact_Sounds/Demolition/Metal/Metal_Garage_Impact04}}
 {id=WavePlayer_11, properties {soundWaveAssetPtr=/Game/Sounds/Fort_Impact_Sounds/Demolition/Metal/Metal_Garage_Impact05}}
 {id=WavePlayer_5, properties {soundWaveAssetPtr=/Game/Sounds/Fort_Impact_Sounds/Demolition/Metal/Metal_Machine_Impact03}}
 {id=WavePlayer_7, properties {soundWaveAssetPtr=/Game/Sounds/Fort_Impact_Sounds/Demolition/Metal/Metal_Machine_Impact04}}
 {id=WavePlayer_8, properties {soundWaveAssetPtr=/Game/Sounds/Fort_Impact_Sounds/Demolition/Metal/Metal_Garage_Impact02}}
 {id=WavePlayer_9, properties {soundWaveAssetPtr=/Game/Sounds/Fort_Impact_Sounds/Demolition/Metal/Metal_Garage_Impact03}}
]
properties {
 attenuationSettings=/Game/Sounds/Attenuation/Fort_Demolition_Hit_Small_Sized_Object_Att
 bOverrideConcurrency=true
 maxDistance=4000
 soundClassObject=&Fort_Impacts
 volumeMultiplier=0.64
 concurrencyOverrides {resolutionRule=PreventNew}
}
