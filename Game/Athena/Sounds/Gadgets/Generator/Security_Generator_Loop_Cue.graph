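# Looping sound cue for the Athena security generator gadget's electrical hum.
# A Mixer layers two looping Battle Lab generator waves at equal volume, with overall volume and pitch slightly lowered.
# It plays through the ambient Athena sound class with a generator attenuation asset for short-range falloff.
# Concurrency allows one instance per owner and prevents new copies from starting.
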SoundCue /Game/Athena/Sounds/Gadgets/Generator/Security_Generator_Loop_Cue
root=Mixer_0
nodes [
 {id=Mixer_0, children [WavePlayer_1,WavePlayer_0], properties {inputVolume [1,1]}}
 {id=WavePlayer_0, properties {bLooping=true, soundWaveAssetPtr=/Game/Sounds/BattleLab/Generator/Electric_Generator_Loop_03}}
 {id=WavePlayer_1, properties {bLooping=true, soundWaveAssetPtr=/Game/Sounds/BattleLab/Generator/Electric_Generator_Loop_01}}
]
properties {
 attenuationSettings=&/Electric_Generator_Att
 maxDistance=900
 pitchMultiplier=0.9
 soundClassObject=&Fort_Ambient_Athena
 volumeMultiplier=0.728
 concurrencyOverrides {bLimitToOwner=true, maxCount=1, resolutionRule=PreventNew}
}
