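# Sound cue for the electrical sparks of the Security Generator gadget in Athena, routed to the Athena ambient sound class.
# A Random node picks one of four equally weighted spark waves and feeds a Modulator at the root.
# Overridden attenuation gives short-range falloff with occlusion filtering and listener focus that lowers unfocused volume.
# A concurrency override caps it at two simultaneous instances, stopping the oldest.
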SoundCue /Game/Athena/Sounds/Gadgets/Generator/Security_Generator_Sparks_Cue
root=Modulator_0
nodes [
 {id=Modulator_0, children [Random_0]}
 {id=Random_0, children [WavePlayer_2,WavePlayer_3,WavePlayer_4,WavePlayer_5], properties {weights [1,1,1,1]}}
 {id=WavePlayer_2, properties {soundWaveAssetPtr=&/Generator_Sparks_04}}
 {id=WavePlayer_3, properties {soundWaveAssetPtr=&/Generator_Sparks_01}}
 {id=WavePlayer_4, properties {soundWaveAssetPtr=&/Generator_Sparks_02}}
 {id=WavePlayer_5, properties {soundWaveAssetPtr=&/Generator_Sparks_03}}
]
properties {
 attenuationSettings=&/Electric_Generator_Att
 bOverrideAttenuation=true
 bOverrideConcurrency=true
 maxDistance=900
 soundClassObject=&Fort_Ambient_Athena
 volumeMultiplier=0.728
 attenuationOverrides {
  bEnableFocusInterpolation=true
  bEnableListenerFocus=true
  bEnableOcclusion=true
  falloffDistance=900
  focusAttackInterpSpeed=0.5
  focusReleaseInterpSpeed=0.5
  nonFocusVolumeAttenuation=0.5
  occlusionLowPassFilterFrequency=5000
  occlusionTraceChannel=ECC_GameTraceChannel11
  occlusionVolumeAttenuation=0.8
  attenuationShapeExtents {x=0}
 }
 concurrencyOverrides {maxCount=2, resolutionRule=StopOldest}
}
