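# Sound cue for a security generator gadget starting up, classed as Athena ambient audio.
# A mixer layers a pitched-up electric generator start with a flashlight start click after a tiny delay and a louder, later turn-on sound with slight random pitch variation.
# Borrowed Creative small-toy generator attenuation spatializes it, and a concurrency override rejects new instances once the limit is reached.
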SoundCue /Game/Athena/Sounds/Gadgets/Generator/Security_Generator_Start_Cue
root=Mixer_0
nodes [
 {id=Delay_0, children [Modulator_0], properties {delayMax=0.3, delayMin=0.3}}
 {id=Delay_1, children [WavePlayer_0], properties {delayMax=0.02, delayMin=0.02}}
 {id=Mixer_0, children [Delay_1,Delay_0,Modulator_1], properties {inputVolume [1,3,1.5]}}
 {id=Modulator_0, children [WavePlayer_2], properties {pitchMax=1}}
 {id=Modulator_1, children [WavePlayer_1], properties {pitchMax=1.1, pitchMin=1.1}}
 {id=WavePlayer_0, properties {soundWaveAssetPtr=/Game/Sounds/Creative/Gadgets/Flashlight/Flashlight_Start01}}
 {id=WavePlayer_1, properties {soundWaveAssetPtr=/Game/Sounds/BattleLab/Generator/Electric_Generator_Start_01}}
 {id=WavePlayer_2, properties {soundWaveAssetPtr=/Game/Sounds/BattleLab/Generator/Electric_Generator_TurnOn_02}}
]
properties {
 attenuationSettings=/Game/Sounds/Creative/Attenuations/Fort_Creative_SmallToy_Generator_Att
 bOverrideConcurrency=true
 maxDistance=1400
 soundClassObject=&Fort_Ambient_Athena
 volumeMultiplier=1.247
 concurrencyOverrides {resolutionRule=PreventNew}
}
