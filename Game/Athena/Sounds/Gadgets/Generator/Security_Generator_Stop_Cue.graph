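# Sound cue for a security generator stopping, layering electrical shutdown and metal destruction sounds under the impacts sound class.
# A mixer blends a random metal garage death hit, a pitched-up generator stop, a briefly delayed pitched-up machine power-down and a later generator turn-off.
# A root modulator wraps the whole mix.
# It uses medium-object destruction attenuation and a concurrency override that prevents new instances once the limit is reached.
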SoundCue /Game/Athena/Sounds/Gadgets/Generator/Security_Generator_Stop_Cue
root=Modulator_1
nodes [
 {id=Delay_0, children [Modulator_3], properties {delayMax=0.1, delayMin=0.1}}
 {id=Delay_1, children [WavePlayer_2], properties {delayMax=0.85, delayMin=0.85}}
 {id=Mixer_0, children [Random_1,Delay_0,Modulator_2,Delay_1], properties {inputVolume [0.7,1,2,1]}}
 {id=Modulator_1, children [Mixer_0]}
 {id=Modulator_2, children [WavePlayer_0], properties {pitchMax=1.1, pitchMin=1.1}}
 {id=Modulator_3, children [WavePlayer_6], properties {pitchMax=1.3, pitchMin=1.3}}
 {id=Random_1, children [WavePlayer_13, WavePlayer_14], properties {weights [1,1]}}
 {id=WavePlayer_0, properties {soundWaveAssetPtr=/Game/Sounds/BattleLab/Generator/Electric_Generator_Stop_01}}
 {id=WavePlayer_13, properties {soundWaveAssetPtr=/Game/Sounds/Fort_Impact_Sounds/Demolition/Metal/Metal_Garage_Death01}}
 {id=WavePlayer_14, properties {soundWaveAssetPtr=/Game/Sounds/Fort_Impact_Sounds/Demolition/Metal/Metal_Garage_Death02}}
 {id=WavePlayer_2, properties {soundWaveAssetPtr=/Game/Sounds/BattleLab/Generator/Electric_Generator_TurnOff_01}}
 {id=WavePlayer_6, properties {soundWaveAssetPtr=/Game/Athena/Sounds/SC_Machine/SC_Machine_Power_Down}}
]
properties {
 attenuationSettings=/Game/Sounds/Attenuation/Fort_Demolition_Destroy_Medium_Sized_Object_Att
 bOverrideConcurrency=true
 maxDistance=4000
 soundClassObject=&Fort_Impacts
 volumeMultiplier=0.64
 concurrencyOverrides {resolutionRule=PreventNew}
}
